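# Sound cue for the feet-freeze effect of the Fortnite ice trap.
# A Random node picks one of three equally weighted feet-freeze wave variations, then a Modulator adds very slight pitch variation.
# Playback runs at reduced volume through small melee weapon attenuation and a weapon melee sound class, with a global blaster concurrency set assigned.
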SoundCue /FNGameplayCues/Sounds/Fort_Traps/Ice_Trap/IceTrap_FeetFreeze_Cue
root=Modulator_0
nodes [
 {id=Modulator_0, children [Random_1], properties {pitchMax=1, pitchMin=0.99}}
 {id=Random_1, children [WavePlayer_0,WavePlayer_5,WavePlayer_6], properties {weights [1,1,1]}}
 {id=WavePlayer_0, properties {soundWaveAssetPtr=&/IceTrap_FeetFreeze_01}}
 {id=WavePlayer_5, properties {soundWaveAssetPtr=&/IceTrap_FeetFreeze_02}}
 {id=WavePlayer_6, properties {soundWaveAssetPtr=&/IceTrap_FeetFreeze_03}}
]
properties {
 attenuationSettings=&Fort_Weapon_Melee_Sml_Att
 maxDistance=2800
 soundClassObject=&Fort_WeaponMelee_PSM
 volumeMultiplier=0.5
 concurrencyOverrides {resolutionRule=PreventNew}
 concurrencySet [/Game/Sounds/Concurrency_Rules/Blaster_Global_Concurrency]
}
