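# Sound cue for Ice Trap footsteps in Fortnite, using third-person enemy foley class and Athena walk-footstep attenuation at reduced volume.
# A Quality Level node selects between random pools of seven, four or two footstep waves, with the two-wave pool shared by the last two slots and excluded from branch culling.
# A root Modulator adds slight random pitch and volume variation to each play.
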SoundCue /FNGameplayCues/Sounds/Fort_Traps/Ice_Trap/IceTrap_Footsteps_Cue
root=Modulator_4
nodes [
 {id=Modulator_4, children [QualityLevel_0], properties {pitchMax=1.1, pitchMin=0.9, volumeMax=1.1, volumeMin=0.9}}
 {id=QualityLevel_0, children [Random_1,Random_0,Random_2,Random_2]}
 {id=Random_0, children [WavePlayer_0,WavePlayer_1,WavePlayer_2,WavePlayer_3], properties {weights [1,1,1,1]}}
 {
  id=Random_1
  children [WavePlayer_5, WavePlayer_6, WavePlayer_7, WavePlayer_8, WavePlayer_9, WavePlayer_10, WavePlayer_11]
  properties {weights [1,1,1,1,1,1,1]}
 }
 {id=Random_2, children [WavePlayer_4, WavePlayer_12], properties {bShouldExcludeFromBranchCulling=true, weights [1,1]}}
 {id=WavePlayer_0, properties {soundWaveAssetPtr=&/IceTrap_Footsteps_V1_01}}
 {id=WavePlayer_1, properties {soundWaveAssetPtr=&/IceTrap_Footsteps_V1_02}}
 {id=WavePlayer_10, properties {soundWaveAssetPtr=&/IceTrap_Footsteps_V1_06}}
 {id=WavePlayer_11, properties {soundWaveAssetPtr=&/IceTrap_Footsteps_V1_07}}
 {id=WavePlayer_12, properties {soundWaveAssetPtr=&IceTrap_Footsteps_V1_02}}
 {id=WavePlayer_2, properties {soundWaveAssetPtr=&/IceTrap_Footsteps_V1_03}}
 {id=WavePlayer_3, properties {soundWaveAssetPtr=&/IceTrap_Footsteps_V1_04}}
 {id=WavePlayer_4, properties {soundWaveAssetPtr=&IceTrap_Footsteps_V1_01}}
 {id=WavePlayer_5, properties {soundWaveAssetPtr=&IceTrap_Footsteps_V1_01}}
 {id=WavePlayer_6, properties {soundWaveAssetPtr=&IceTrap_Footsteps_V1_02}}
 {id=WavePlayer_7, properties {soundWaveAssetPtr=&IceTrap_Footsteps_V1_03}}
 {id=WavePlayer_8, properties {soundWaveAssetPtr=&IceTrap_Footsteps_V1_04}}
 {id=WavePlayer_9, properties {soundWaveAssetPtr=&/IceTrap_Footsteps_V1_05}}
]
properties {
 attenuationSettings=/Game/Sounds/Attenuation/Fort_Player_Footsteps_3P_Walk_Athena_Att
 maxDistance=3584
 soundClassObject=&Fort_Player_Foley_3P_Enemy_Athena
 volumeMultiplier=0.2
}
