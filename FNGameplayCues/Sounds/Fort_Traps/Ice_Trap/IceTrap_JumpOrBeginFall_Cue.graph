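# Sound cue for the ice trap lift-off, played when a character affected by the trap jumps or begins to fall, as its name indicates.
# It picks one of four equally weighted lift-off waves at random, then applies slight random pitch and volume variation through a modulator.
# Playback is quieted and uses third-person walking footstep attenuation with a melee weapon sound class.
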SoundCue /FNGameplayCues/Sounds/Fort_Traps/Ice_Trap/IceTrap_JumpOrBeginFall_Cue
root=Modulator_0
nodes [
 {id=Modulator_0, children [Random_0], properties {pitchMax=1.1, pitchMin=0.9, volumeMax=1.1, volumeMin=0.9}}
 {id=Random_0, children [WavePlayer_0,WavePlayer_1,WavePlayer_2,WavePlayer_3], properties {weights [1,1,1,1]}}
 {id=WavePlayer_0, properties {soundWaveAssetPtr=&/IceTrap_LiftOff_04}}
 {id=WavePlayer_1, properties {soundWaveAssetPtr=&/IceTrap_LiftOff_01}}
 {id=WavePlayer_2, properties {soundWaveAssetPtr=&/IceTrap_LiftOff_02}}
 {id=WavePlayer_3, properties {soundWaveAssetPtr=&/IceTrap_LiftOff_03}}
]
properties {
 attenuationSettings=/Game/Sounds/Attenuation/Fort_Player_Footsteps_3P_Walk_Athena_Att
 maxDistance=3584
 soundClassObject=&Fort_WeaponMelee_PSM
 volumeMultiplier=0.32
}
