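# IceTrap_PlayerLand_Cue is a SoundCue for a player landing on an ice trap, playing slide impact sounds.
# A Random node picks among five equally weighted IceTrap_SlideImpact waves and feeds a Modulator at the root.
# It borrows small melee weapon attenuation and a melee sound class, plays at reduced volume, and overrides concurrency so new instances are prevented once the limit is reached.
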SoundCue /FNGameplayCues/Sounds/Fort_Traps/Ice_Trap/IceTrap_PlayerLand_Cue
root=Modulator_2
nodes [
 {id=Modulator_2, children [Random_0]}
 {id=Random_0, children [WavePlayer_0,WavePlayer_1,WavePlayer_2,WavePlayer_3,WavePlayer_9], properties {weights [1,1,1,1,1]}}
 {id=WavePlayer_0, properties {soundWaveAssetPtr=&/IceTrap_SlideImpact_01}}
 {id=WavePlayer_1, properties {soundWaveAssetPtr=&/IceTrap_SlideImpact_02}}
 {id=WavePlayer_2, properties {soundWaveAssetPtr=&/IceTrap_SlideImpact_03}}
 {id=WavePlayer_3, properties {soundWaveAssetPtr=&/IceTrap_SlideImpact_04}}
 {id=WavePlayer_9, properties {soundWaveAssetPtr=&/IceTrap_SlideImpact_05}}
]
properties {
 attenuationSettings=&Fort_Weapon_Melee_Sml_Att
 bOverrideConcurrency=true
 maxDistance=2800
 soundClassObject=&Fort_WeaponMelee_PSM
 volumeMultiplier=0.6
 attenuationOverrides {falloffDistance=2600}
 concurrencyOverrides {resolutionRule=PreventNew}
}
