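# Looping sound cue for a vehicle sliding on the Ice Trap, routed through the vehicle sound class.
# A quality-level root plays the looping vehicle slide wave for its first two levels and reuses the looping player slide wave for the last two.
# It borrows the golf cart's close movement attenuation for spatial falloff.
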SoundCue /FNGameplayCues/Sounds/Fort_Traps/Ice_Trap/IceTrap_SlideLoop_Vehicle_Cue
root=QualityLevel_0
nodes [
 {id=QualityLevel_0, children [WavePlayer_1,WavePlayer_1,WavePlayer_0,WavePlayer_0]}
 {id=WavePlayer_0, properties {bLooping=true, soundWaveAssetPtr=/Game/Sounds/Fort_Traps/Ice_Trap/IceTrap_SlideLoop_Player_01}}
 {id=WavePlayer_1, properties {bLooping=true, soundWaveAssetPtr=&/IceTrap_SlideLoop_Vehicle_01}}
]
properties {
 attenuationSettings=/Game/Sounds/Attenuation/Fort_GolfCart_Movement_Close_Att
 maxDistance=5300
 soundClassObject=&Fort_Vehicles
 volumeMultiplier=1.1
 concurrencyOverrides {bLimitToOwner=true}
}
